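# Niagara system for the swing trail of the ChiliAbide guitar pickaxe cosmetic.
# Three ribbon emitters (Soft_Trail, ToonTrail, ToonTrailBrown) spawn short-lived anim trails between different bone or socket pairs sharing one index.
# Each fades alpha over particle life, multiplied by System.AlphaFade, and tints RGB through a curve indexed by time of day.
# Alternative DaySequence color modules and AddVelocity are present but disabled.
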NiagaraSystem /SparksCosmetics/Instrument/Guitar/ChiliAbide/Effects/NS_Pickaxe_Guitar_ChiliAbide_SwingTrail
dependencies {
 usedDataInterfaces [&NiagaraDataInterfaceColorCurve, &NiagaraDataInterfaceCurve, &NiagaraDataInterfaceSocketReader]
 usedDynamicInputs [&MergeHelper_FloatSwitch]
 usedModules [
  &SystemState
  &self:SystemSpawnScript.SystemScriptSource.SystemScriptGraph.NiagaraNodeAssignment_2.SetVariables_#1
  &self:SystemSpawnScript.SystemScriptSource.SystemScriptGraph.NiagaraNodeAssignment_3.SetVariables_#2
 ]
}
emitters {
 Soft_Trail {
  emitterSpawnScript [&FN_RibbonIDOffsetUtility()]
  emitterUpdateScript [
   &EmitterState()
   &FN_SpawnAnimTrail(AnimTrailWidthMode=::FromSecond)
  ]
  particleSpawnScript [
   &FN_InitializeAnimTrailBoneIndecies(First Bone or Socket Index=1, Second Bone or Socket Index=0)
   &FN_InitializeAnimTrail(AnimTrailWidth=0.75, Color=<rgba 0.4045 0.4045 0.4045 0.25>, Lifetime=0.18)
   &AddVelocity(Velocity=<500 0 0>) {enabled=false}
  ]
  particleUpdateScript [
   &SolveForcesAndVelocity(Clamp Velocity=false, Limit Acceleration=false)
   &FN_UpdateAnimTrail() {module=FN_UpdateAnimTrail001}
   &ScaleColor(
    Scale Alpha=&Multiply_Float(
     A=&FloatFromCurve(FloatCurve=DataInterface() {propertyValues {Curve {keys [<at 0.1 1 RCIM_Cubic>, {arriveTangent=-2.5031, tangentMode=RCTM_User, time=1, value=0}]}}})
     B=$System.AlphaFade
    )
    Scale RGB=&MakeVectorFromLinearColorRGB(Color=&ColorFromCurve(
      ColorCurve=DataInterface() {propertyValues {
        AlphaCurve {keys [<at 0 1 RCIM_Cubic>]}
        BlueCurve {keys [<at 0 1 RCIM_Cubic>,<at 0.25 0.744>,<at 0.5 1>,<at 0.75 0.744>,<at 1 1 RCIM_Cubic>]}
        GreenCurve {keys [<at 0 0.5531 RCIM_Cubic>,<at 0.25 0.6155>,<at 0.5 1>,<at 0.75 0.6155>,<at 1 0.5531 RCIM_Cubic>]}
        RedCurve {keys [<at 0 0.3148 RCIM_Cubic>,<at 0.25 0.8174>,<at 0.5 1>,<at 0.75 0.8174>,<at 1 0.3148 RCIM_Cubic>]}
       }}
      CurveIndex=Linked() {linkedVariable {name=NPC.FortniteNPC.TimeOfDay_0-1, type=NiagaraFloat}}
     ))
    ScaleA=true
    ScaleRGB=true
   )
   &ScaleColor(
    Scale Alpha=&Multiply_Float(
     A=&FloatFromCurve(FloatCurve=DataInterface() {propertyValues {Curve {keys [<at 0.1 1 RCIM_Cubic>, {arriveTangent=-2.5031, tangentMode=RCTM_User, time=1, value=0}]}}})
     B=$System.AlphaFade
    )
    Scale RGB=&MakeVectorFromLinearColorRGB(Color=/Game/Effects/Niagara/DynamicInputs/TimeOfDay/DaySequenceTimeOfDayColor(Linear Color:LinearColor=<rgba 1 1 1 1>, Time Of Day Influence:NiagaraFloat=1))
    ScaleA=true
    ScaleRGB=true
   ) {enabled=false, module=ScaleColor001}
  ]
  rendererClasses [&NiagaraRibbonRendererProperties]
  renderers [{rendererClass=&NiagaraRibbonRendererProperties}]
 }
 ToonTrail {
  emitterSpawnScript [&FN_RibbonIDOffsetUtility()]
  emitterUpdateScript [
   &EmitterState()
   &FN_SpawnAnimTrail(AnimTrailWidthMode=::FromSecond)
  ]
  particleSpawnScript [
   &FN_InitializeAnimTrailBoneIndecies(First Bone or Socket Index=1, Second Bone or Socket Index=2)
   &FN_InitializeAnimTrail(AnimTrailWidth=0.75, Color=<rgba 0.3573 0.3764 0.4198 0.95>, Lifetime=0.16)
   &AddVelocity(Velocity=<500 0 0>) {enabled=false}
  ]
  particleUpdateScript [
   &SolveForcesAndVelocity(Clamp Velocity=false, Limit Acceleration=false)
   &FN_UpdateAnimTrail() {module=FN_UpdateAnimTrail001}
   &ScaleColor(
    Scale Alpha=&Multiply_Float(
     A=&FloatFromCurve(FloatCurve=DataInterface() {propertyValues {Curve {keys [<at 0.1 1 RCIM_Cubic>, {arriveTangent=-2.5031, tangentMode=RCTM_User, time=1, value=0}]}}})
     B=$System.AlphaFade
    )
    Scale RGB=&MakeVectorFromLinearColorRGB(Color=&ColorFromCurve(
      ColorCurve=DataInterface() {propertyValues {
        AlphaCurve {keys [<at 0 1 RCIM_Cubic>]}
        BlueCurve {keys [<at 0 1 RCIM_Cubic>,<at 0.25 0.744>,<at 0.5 1>,<at 0.75 0.744>,<at 1 1 RCIM_Cubic>]}
        GreenCurve {keys [<at 0 0.5531 RCIM_Cubic>,<at 0.25 0.6155>,<at 0.5 1>,<at 0.75 0.6155>,<at 1 0.5531 RCIM_Cubic>]}
        RedCurve {keys [<at 0 0.3148 RCIM_Cubic>,<at 0.25 0.8174>,<at 0.5 1>,<at 0.75 0.8174>,<at 1 0.3148 RCIM_Cubic>]}
       }}
      CurveIndex=Linked() {linkedVariable {name=NPC.FortniteNPC.TimeOfDay_0-1, type=NiagaraFloat}}
     ))
    ScaleA=true
    ScaleRGB=true
   )
   &ScaleColor(
    Scale Alpha=&Multiply_Float(
     A=&FloatFromCurve(FloatCurve=DataInterface() {propertyValues {Curve {keys [<at 0.1 1 RCIM_Cubic>, {arriveTangent=-2.5031, tangentMode=RCTM_User, time=1, value=0}]}}})
     B=$System.AlphaFade
    )
    Scale RGB=&MakeVectorFromLinearColorRGB(Color=&DaySequenceTimeOfDayColor(Linear Color:LinearColor=<rgba 1 1 1 1>, Time Of Day Influence:NiagaraFloat=1))
    ScaleA=true
    ScaleRGB=true
   ) {enabled=false, module=ScaleColor001}
  ]
  rendererClasses [&NiagaraRibbonRendererProperties]
  renderers [{rendererClass=&NiagaraRibbonRendererProperties}]
 }
 ToonTrailBrown {
  emitterSpawnScript [&FN_RibbonIDOffsetUtility()]
  emitterUpdateScript [
   &EmitterState()
   &FN_SpawnAnimTrail()
  ]
  particleSpawnScript [
   &FN_InitializeAnimTrailBoneIndecies(First Bone or Socket Index=1, Second Bone or Socket Index=4)
   &FN_InitializeAnimTrail(AnimTrailWidth=0.6, Color=<rgba 0.3573 0.3764 0.4198 0.95>, Lifetime=0.1)
   &AddVelocity(Velocity=<500 0 0>) {enabled=false}
  ]
  particleUpdateScript [
   &SolveForcesAndVelocity(Clamp Velocity=false, Limit Acceleration=false)
   &FN_UpdateAnimTrail() {module=FN_UpdateAnimTrail001}
   &ScaleColor(
    Scale Alpha=&Multiply_Float(
     A=&FloatFromCurve(FloatCurve=DataInterface() {propertyValues {Curve {keys [<at 0.1 1 RCIM_Cubic>, {arriveTangent=-2.5031, tangentMode=RCTM_User, time=1, value=0}]}}})
     B=$System.AlphaFade
    )
    Scale RGB=&MakeVectorFromLinearColorRGB(Color=&ColorFromCurve(
      ColorCurve=DataInterface() {propertyValues {
        AlphaCurve {keys [<at 0 1 RCIM_Cubic>]}
        BlueCurve {keys [<at 0 1 RCIM_Cubic>,<at 0.25 0.744>,<at 0.5 1>,<at 0.75 0.744>,<at 1 1 RCIM_Cubic>]}
        GreenCurve {keys [<at 0 0.5531 RCIM_Cubic>,<at 0.25 0.6155>,<at 0.5 1>,<at 0.75 0.6155>,<at 1 0.5531 RCIM_Cubic>]}
        RedCurve {keys [<at 0 0.3148 RCIM_Cubic>,<at 0.25 0.8174>,<at 0.5 1>,<at 0.75 0.8174>,<at 1 0.3148 RCIM_Cubic>]}
       }}
      CurveIndex=Linked() {linkedVariable {name=NPC.FortniteNPC.TimeOfDay_0-1, type=NiagaraFloat}}
     ))
    ScaleA=true
    ScaleRGB=true
   )
   &ScaleColor(
    Scale Alpha=&Multiply_Float(
     A=&FloatFromCurve(FloatCurve=DataInterface() {propertyValues {Curve {keys [<at 0.1 1 RCIM_Cubic>, {arriveTangent=-2.5031, tangentMode=RCTM_User, time=1, value=0}]}}})
     B=$System.AlphaFade
    )
    Scale RGB=&MakeVectorFromLinearColorRGB(Color=&DaySequenceTimeOfDayColor(Linear Color:LinearColor=<rgba 1 1 1 1>, Time Of Day Influence:NiagaraFloat=1))
    ScaleA=true
    ScaleRGB=true
   ) {enabled=false, module=ScaleColor001}
  ]
  rendererClasses [&NiagaraRibbonRendererProperties]
  renderers [{rendererClass=&NiagaraRibbonRendererProperties}]
 }
}
summary {
 systemName=NS_Pickaxe_Guitar_ChiliAbide_SwingTrail
 emitters [
  {emitterName=Soft_Trail, rendererClasses [&NiagaraRibbonRendererProperties]}
  {emitterName=ToonTrail, rendererClasses [&NiagaraRibbonRendererProperties]}
  {emitterName=ToonTrailBrown, rendererClasses [&NiagaraRibbonRendererProperties]}
 ]
 userVariables [
  User.RibbonIDOffset {type=NiagaraInt32}
  User.SkeletalMesh {
   type=NiagaraDataInterfaceSkeletalMesh
   defaultValue {
    struct=NiagaraExt_VariableValue_DataInterface
    value {dataInterface=&self:NiagaraDataInterfaceSkeletalMesh_0}
   }
  }
  User.SwingSectionIndex {type=NiagaraInt32}
 ]
}
